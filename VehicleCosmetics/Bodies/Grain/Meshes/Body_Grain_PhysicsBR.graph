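# Physics asset for the Grain vehicle body cosmetic, named as the BR physics variant for that body mesh.
# It holds a single body on the root bone, with no collision shapes listed for it.
# It has no constraints, so it sets up no jointed or multi-body simulation.
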PhysicsAsset /VehicleCosmetics/Bodies/Grain/Meshes/Body_Grain_PhysicsBR
bodies {root []}
constraints []
