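# Physics asset for the Grain vehicle body mesh, providing its collision shape.
# It holds a single body on the root bone with one box that approximates the whole body, offset slightly forward and raised above the origin.
# It has no constraints, so the body acts as one rigid piece with no articulated parts.
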PhysicsAsset /VehicleCosmetics/Bodies/Grain/Meshes/Body_Grain_PhysicsDM
bodies {root [{shapeType=Box, extentX=444, extentY=216, extentZ=116, shapeName=root_box, center=<23.5 0 93>}]}
constraints []
