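# Physics asset for the Grain vehicle body's hit box skeletal mesh.
# It holds a single box shape on the root bone, offset forward and upward from the bone origin.
# There are no other bodies and no constraints, so the whole hit box is one rigid volume.
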PhysicsAsset /VehicleCosmetics/Bodies/Grain/Meshes/SK_Body_Grain_HitBox
bodies {root [{shapeType=Box, extentX=377.6236, extentY=269.4381, extentZ=115.709, shapeName=root_box, center=<44.4021 0 66.416>}]}
constraints []
