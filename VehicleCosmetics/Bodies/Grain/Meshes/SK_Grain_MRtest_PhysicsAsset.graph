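# Physics asset for the Grain vehicle cosmetic skeletal mesh (MRtest variant), giving each listed bone a single capsule body.
# The body, front steering bones and rear knuckles constrain to a tiny frame capsule.
# Glass, steering wheel and axle pivots hang off the body, while wheel discs attach to their steering bone or knuckle.
# All constraints share identical swing and twist limit angles, with no per-joint tuning of those angles.
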PhysicsAsset /VehicleCosmetics/Bodies/Grain/Meshes/SK_Grain_MRtest_PhysicsAsset
bodies {
 axle_pivot_bk_l [{shapeType=Capsule, length=29.6155, radius=27.5905, center=<-0.2488 -24.6971 14.1803>, rotation=<rot -14.7784 -90 -0.8952>}]
 axle_pivot_bk_r [{shapeType=Capsule, length=29.6155, radius=27.5905, center=<-0.2488 24.6977 14.1803>, rotation=<rot 14.7784 -90 -0.8951>}]
 axle_pivot_fr_l [{shapeType=Capsule, length=26.8293, radius=25.3426, center=<-0.0878 -21.7865 12.8025>, rotation=<rot -13.0018 -90 -0.6488>}]
 axle_pivot_fr_r [{shapeType=Capsule, length=26.8293, radius=25.3426, center=<-0.0878 21.7871 12.8025>, rotation=<rot 13.002 -90 -0.6488>}]
 body [{shapeType=Capsule, length=242.395, radius=123.5021, center=<28.4454 -0.0468 29.0458>, rotation=<rot 0 -90.022 -85.9549>}]
 frame [{shapeType=Capsule, length=0.505, radius=0.505, rotation=<rot 0 -90 0>}]
 glass_b [{shapeType=Capsule, length=75.9884, radius=14.9881, center=<12.0614 0 16.458>, rotation=<rot 0 0 -90>}]
 glass_c [{shapeType=Capsule, length=80.3148, radius=39.5991, center=<-38.6225 0 21.0103>, rotation=<rot 0 0 -90>}]
 glass_l [{shapeType=Capsule, length=94.4153, radius=19.0224, center=<-39.5503 7.681 24.7863>, rotation=<rot 0 -88.8087 -85.28>}]
 glass_r [{shapeType=Capsule, length=94.4153, radius=19.0224, center=<-39.5503 -7.6809 24.7862>, rotation=<rot 0 -91.1913 -85.28>}]
 steering_knuckle_bk_l [{shapeType=Capsule, length=17.7695, radius=13.8007, center=<6.6259 -1.532 4.7992>, rotation=<rot 0 66.2419 -49.1927>}]
 steering_knuckle_bk_r [{shapeType=Capsule, length=17.7695, radius=13.8007, center=<6.626 1.5326 4.7992>, rotation=<rot 0 113.7581 -49.1926>}]
 steering_wheel [{shapeType=Capsule, length=20.4119, radius=16.7126, center=<2.9924 -0.1155 -1.2394>, rotation=<rot 0 -179.9471 -75.7769>}]
 wheel_disc_bk_l [{shapeType=Capsule, length=22.8622, radius=22.8621, center=<-0.0237 -1.6814 0.1669>, rotation=<rot 0 -90 0.0009>}]
 wheel_disc_bk_r [{shapeType=Capsule, length=22.8622, radius=22.8621, center=<-0.0237 1.6819 0.1669>, rotation=<rot 0 -90 0.0009>}]
 wheel_disc_fr_l [{shapeType=Capsule, length=22.8622, radius=22.8622, center=<0.0986 -1.6178 0.1656>, rotation=<rot 0 -90 -0.0001>}]
 wheel_disc_fr_r [{shapeType=Capsule, length=22.8622, radius=22.8622, center=<0.0986 1.6167 0.1657>, rotation=<rot 0 -90 0>}]
 wheel_steering_fr_l [{shapeType=Capsule, length=17.7695, radius=13.8007, center=<6.7483 -2.1748 4.7983>, rotation=<rot 0 66.242 -49.1928>}]
 wheel_steering_fr_r [{shapeType=Capsule, length=17.7695, radius=13.8007, center=<6.7483 2.1737 4.7984>, rotation=<rot 0 113.758 -49.1928>}]
}
constraints [
 {bone1Name=body, bone2Name=frame, swing1LimitDegrees=45, swing2LimitDegrees=45, twistLimitDegrees=45}
 {bone1Name=glass_c, bone2Name=body, swing1LimitDegrees=45, swing2LimitDegrees=45, twistLimitDegrees=45}
 {bone1Name=glass_b, bone2Name=body, swing1LimitDegrees=45, swing2LimitDegrees=45, twistLimitDegrees=45}
 {bone1Name=glass_l, bone2Name=body, swing1LimitDegrees=45, swing2LimitDegrees=45, twistLimitDegrees=45}
 {bone1Name=glass_r, bone2Name=body, swing1LimitDegrees=45, swing2LimitDegrees=45, twistLimitDegrees=45}
 {bone1Name=steering_wheel, bone2Name=body, swing1LimitDegrees=45, swing2LimitDegrees=45, twistLimitDegrees=45}
 {bone1Name=axle_pivot_fr_l, bone2Name=body, swing1LimitDegrees=45, swing2LimitDegrees=45, twistLimitDegrees=45}
 {bone1Name=axle_pivot_fr_r, bone2Name=body, swing1LimitDegrees=45, swing2LimitDegrees=45, twistLimitDegrees=45}
 {bone1Name=axle_pivot_bk_l, bone2Name=body, swing1LimitDegrees=45, swing2LimitDegrees=45, twistLimitDegrees=45}
 {bone1Name=axle_pivot_bk_r, bone2Name=body, swing1LimitDegrees=45, swing2LimitDegrees=45, twistLimitDegrees=45}
 {bone1Name=wheel_steering_fr_l, bone2Name=frame, swing1LimitDegrees=45, swing2LimitDegrees=45, twistLimitDegrees=45}
 {bone1Name=wheel_disc_fr_l, bone2Name=wheel_steering_fr_l, swing1LimitDegrees=45, swing2LimitDegrees=45, twistLimitDegrees=45}
 {bone1Name=wheel_steering_fr_r, bone2Name=frame, swing1LimitDegrees=45, swing2LimitDegrees=45, twistLimitDegrees=45}
 {bone1Name=wheel_disc_fr_r, bone2Name=wheel_steering_fr_r, swing1LimitDegrees=45, swing2LimitDegrees=45, twistLimitDegrees=45}
 {bone1Name=steering_knuckle_bk_l, bone2Name=frame, swing1LimitDegrees=45, swing2LimitDegrees=45, twistLimitDegrees=45}
 {bone1Name=wheel_disc_bk_l, bone2Name=steering_knuckle_bk_l, swing1LimitDegrees=45, swing2LimitDegrees=45, twistLimitDegrees=45}
 {bone1Name=steering_knuckle_bk_r, bone2Name=frame, swing1LimitDegrees=45, swing2LimitDegrees=45, twistLimitDegrees=45}
 {bone1Name=wheel_disc_bk_r, bone2Name=steering_knuckle_bk_r, swing1LimitDegrees=45, swing2LimitDegrees=45, twistLimitDegrees=45}
]
